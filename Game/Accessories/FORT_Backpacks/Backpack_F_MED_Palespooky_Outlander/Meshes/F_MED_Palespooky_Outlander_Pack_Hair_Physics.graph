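# Physics asset for the Palespooky Outlander backpack, giving its dreadlock hair and skirt bones simulated secondary motion.
# Capsule chains for each dreadlock hang from a head sphere with limited swing and locked twist, though strands 05 and 08 lock rigidly at the root.
# The head is fully locked to a torso capsule.
# A skirt hub swings from the torso with locked twist and carries four side capsules on angularly free joints.
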PhysicsAsset /Game/Accessories/FORT_Backpacks/Backpack_F_MED_Palespooky_Outlander/Meshes/F_MED_Palespooky_Outlander_Pack_Hair_Physics
bodies {
 dyn_dreadlock_01_A [{shapeType=Capsule, length=4, radius=0.7, center=<3.2927 0 0.9051>, rotation=<rot 90 0 0>}]
 dyn_dreadlock_01_B [{shapeType=Capsule, length=2.2644, radius=1, center=<2.3014 0 0.7087>, rotation=<rot 90 0 0>}]
 dyn_dreadlock_01_C [{shapeType=Capsule, length=2.2713, radius=1, center=<2.2126 0 0>, rotation=<rot 90 0 0>}]
 dyn_dreadlock_01_D [{shapeType=Capsule, length=1.7186, radius=1, center=<1.6726 0 0>, rotation=<rot 90 0 0>}]
 dyn_dreadlock_02_B [{shapeType=Capsule, length=2, radius=1, center=<1.5 0 0>, rotation=<rot 90 0 0>}]
 dyn_dreadlock_02_C [{shapeType=Capsule, length=2, radius=1, center=<1.5 0 0>, rotation=<rot 90 0 0>}]
 dyn_dreadlock_02_D [{shapeType=Capsule, length=2, radius=1, center=<1.5 0 0>, rotation=<rot 90 0 0>}]
 dyn_dreadlock_03_B [{shapeType=Capsule, length=2.9705, radius=1, center=<2.6933 0 0.8689>, rotation=<rot 90 0 0>}]
 dyn_dreadlock_03_C [{shapeType=Capsule, length=1.8606, radius=1, center=<1.8916 0 0.8278>, rotation=<rot 90 0 0>}]
 dyn_dreadlock_03_D [{shapeType=Capsule, length=3.1917, radius=1, center=<2.7611 0 0.5791>, rotation=<rot 90 0 0>}]
 dyn_dreadlock_05_B [{shapeType=Capsule, length=2, radius=1, center=<1.8418 0 1.3456>, rotation=<rot 90 0 0>}]
 dyn_dreadlock_05_C [{shapeType=Capsule, length=2, radius=1, center=<2.3738 0 1.3673>, rotation=<rot 86.9467 0.0001 0.0001>}]
 dyn_dreadlock_05_D [{shapeType=Capsule, length=2, radius=1, center=<2.5192 0 0.3816>, rotation=<rot 90 0 0>}]
 dyn_dreadlock_07_B [{shapeType=Capsule, length=2, radius=1, center=<2.5804 0.0001 1.5106>, rotation=<rot 90 0 0>}]
 dyn_dreadlock_07_C [{shapeType=Capsule, length=2, radius=1, center=<1.9519 0.0001 1.5106>, rotation=<rot 90 0 0>}]
 dyn_dreadlock_07_D [{shapeType=Capsule, length=2, radius=1, center=<1.5466 0.0001 1.5106>, rotation=<rot 90 0 0>}]
 dyn_dreadlock_08_B [{shapeType=Capsule, length=2.0999, radius=1, center=<2.3746 0 0.7958>, rotation=<rot 90 0 0>}]
 dyn_dreadlock_08_C [{shapeType=Capsule, length=2.4382, radius=1, center=<0.7896 0 0.6043>, rotation=<rot 90 0 0>}]
 dyn_dreadlock_08_D [{shapeType=Capsule, length=3, radius=1, center=<1.3425 0 0.4297>, rotation=<rot 90 0 0>}]
 dyn_dreadlock_10_B [{shapeType=Capsule, length=3, radius=1, center=<1.6916 0 0>, rotation=<rot 90 0 0>}]
 dyn_dreadlock_10_C [{shapeType=Capsule, length=2.829, radius=1, center=<2.8813 0 0>, rotation=<rot 90 0 0>}]
 dyn_dreadlock_11_B [{shapeType=Capsule, length=2, radius=1, center=<1.5 0 0.2862>, rotation=<rot 90 0 0>}]
 dyn_dreadlock_11_C [{shapeType=Capsule, length=2, radius=1, center=<1.5 0 0>, rotation=<rot 90 0 0>}]
 dyn_dreadlock_12_B [{shapeType=Capsule, length=2.4006, radius=1, center=<2.2791 -0.4502 0.7614>, rotation=<rot 90 0 0>}]
 dyn_dreadlock_12_C [{shapeType=Capsule, length=3.1733, radius=1, center=<3.0325 0 0>, rotation=<rot 90 0 0>}]
 dyn_dreadlock_14_A [{shapeType=Capsule, length=2.1353, radius=1, center=<2.3317 0 1.0014>, rotation=<rot 90 0 0>}]
 dyn_dreadlock_14_B [{shapeType=Capsule, length=2.1289, radius=1, center=<2.0584 0 0.7656>, rotation=<rot 90 0 0>}]
 dyn_dreadlock_14_C [{shapeType=Capsule, length=2.1773, radius=1, center=<2.1516 0 0>, rotation=<rot 90 0 0>}]
 dyn_dreadlock_14_D [{shapeType=Capsule, length=1.7362, radius=1, center=<1.6537 0 0>, rotation=<rot 90 0 0>}]
 dyn_headHairParent [{shapeType=Sphere, radius=6.21, center=<0 0 -0.4895>}]
 dyn_skirt [{shapeType=Capsule, length=6, radius=3, center=<0 0 -1>}]
 dyn_skirtBack [{shapeType=Capsule, length=2, radius=1.5, center=<0 0 -0.4913>}]
 dyn_skirtFront [{shapeType=Capsule, length=2, radius=1.5, center=<0 0 -0.4913>}]
 dyn_skirtLeft [{shapeType=Capsule, length=2, radius=1.5, center=<0 0 -0.4913>}]
 dyn_skirtRight [{shapeType=Capsule, length=2, radius=1.5, center=<0 0 -0.4913>}]
 dyn_torso [{shapeType=Capsule, length=9.796, radius=3.3988, center=<0 0.1526 0.0509>, rotation=<rot 0 0 5.3388>}]
}
constraints [
 {
  bone1Name=dyn_dreadlock_01_B
  bone2Name=dyn_dreadlock_01_A
  swing1LimitDegrees=35
  swing2LimitDegrees=35
  twistLimitDegrees=45
  twistMotion=Locked
 }
 {
  bone1Name=dyn_dreadlock_01_C
  bone2Name=dyn_dreadlock_01_B
  swing1LimitDegrees=35
  swing2LimitDegrees=35
  twistLimitDegrees=45
  twistMotion=Locked
 }
 {
  bone1Name=dyn_dreadlock_01_D
  bone2Name=dyn_dreadlock_01_C
  swing1LimitDegrees=35
  swing2LimitDegrees=35
  twistLimitDegrees=45
  twistMotion=Locked
 }
 {
  bone1Name=dyn_dreadlock_03_B
  bone2Name=dyn_headHairParent
  swing1LimitDegrees=35
  swing2LimitDegrees=35
  twistLimitDegrees=45
  twistMotion=Locked
 }
 {
  bone1Name=dyn_dreadlock_03_C
  bone2Name=dyn_dreadlock_03_B
  swing1LimitDegrees=35
  swing2LimitDegrees=35
  twistLimitDegrees=45
  twistMotion=Locked
 }
 {
  bone1Name=dyn_dreadlock_03_D
  bone2Name=dyn_dreadlock_03_C
  swing1LimitDegrees=35
  swing2LimitDegrees=35
  twistLimitDegrees=45
  twistMotion=Locked
 }
 {
  bone1Name=dyn_dreadlock_05_B
  bone2Name=dyn_headHairParent
  swing1LimitDegrees=35
  swing1Motion=Locked
  swing2LimitDegrees=35
  swing2Motion=Locked
  twistLimitDegrees=45
  twistMotion=Locked
 }
 {
  bone1Name=dyn_dreadlock_05_C
  bone2Name=dyn_dreadlock_05_B
  swing1LimitDegrees=35
  swing2LimitDegrees=35
  twistLimitDegrees=45
  twistMotion=Locked
 }
 {
  bone1Name=dyn_dreadlock_05_D
  bone2Name=dyn_dreadlock_05_C
  swing1LimitDegrees=35
  swing2LimitDegrees=35
  twistLimitDegrees=45
  twistMotion=Locked
 }
 {
  bone1Name=dyn_dreadlock_07_B
  bone2Name=dyn_headHairParent
  swing1LimitDegrees=25
  swing2LimitDegrees=25
  twistLimitDegrees=45
  twistMotion=Locked
 }
 {
  bone1Name=dyn_dreadlock_07_C
  bone2Name=dyn_dreadlock_07_B
  swing1LimitDegrees=25
  swing2LimitDegrees=25
  twistLimitDegrees=45
  twistMotion=Locked
 }
 {
  bone1Name=dyn_dreadlock_07_D
  bone2Name=dyn_dreadlock_07_C
  swing1LimitDegrees=25
  swing2LimitDegrees=25
  twistLimitDegrees=45
  twistMotion=Locked
 }
 {
  bone1Name=dyn_dreadlock_08_B
  bone2Name=dyn_headHairParent
  swing1LimitDegrees=25
  swing1Motion=Locked
  swing2LimitDegrees=25
  swing2Motion=Locked
  twistLimitDegrees=45
  twistMotion=Locked
 }
 {
  bone1Name=dyn_dreadlock_08_C
  bone2Name=dyn_dreadlock_08_B
  swing1LimitDegrees=25
  swing2LimitDegrees=25
  twistLimitDegrees=45
  twistMotion=Locked
 }
 {
  bone1Name=dyn_dreadlock_08_D
  bone2Name=dyn_dreadlock_08_C
  swing1LimitDegrees=25
  swing2LimitDegrees=25
  twistLimitDegrees=45
  twistMotion=Locked
 }
 {
  bone1Name=dyn_dreadlock_10_B
  bone2Name=dyn_headHairParent
  swing1LimitDegrees=25
  swing2LimitDegrees=25
  twistLimitDegrees=45
  twistMotion=Locked
 }
 {
  bone1Name=dyn_dreadlock_10_C
  bone2Name=dyn_dreadlock_10_B
  swing1LimitDegrees=25
  swing2LimitDegrees=25
  twistLimitDegrees=45
  twistMotion=Locked
 }
 {
  bone1Name=dyn_dreadlock_12_B
  bone2Name=dyn_headHairParent
  swing1LimitDegrees=35
  swing2LimitDegrees=35
  twistLimitDegrees=45
  twistMotion=Locked
 }
 {
  bone1Name=dyn_dreadlock_12_C
  bone2Name=dyn_dreadlock_12_B
  swing1LimitDegrees=35
  swing2LimitDegrees=35
  twistLimitDegrees=45
  twistMotion=Locked
 }
 {
  bone1Name=dyn_dreadlock_14_B
  bone2Name=dyn_dreadlock_14_A
  swing1LimitDegrees=35
  swing2LimitDegrees=35
  twistLimitDegrees=45
  twistMotion=Locked
 }
 {
  bone1Name=dyn_dreadlock_14_C
  bone2Name=dyn_dreadlock_14_B
  swing1LimitDegrees=35
  swing2LimitDegrees=35
  twistLimitDegrees=45
  twistMotion=Locked
 }
 {
  bone1Name=dyn_dreadlock_14_D
  bone2Name=dyn_dreadlock_14_C
  swing1LimitDegrees=35
  swing2LimitDegrees=35
  twistLimitDegrees=45
  twistMotion=Locked
 }
 {
  bone1Name=dyn_dreadlock_01_A
  bone2Name=dyn_headHairParent
  swing1LimitDegrees=35
  swing2LimitDegrees=35
  twistLimitDegrees=45
  twistMotion=Locked
 }
 {
  bone1Name=dyn_dreadlock_14_A
  bone2Name=dyn_headHairParent
  swing1LimitDegrees=35
  swing2LimitDegrees=35
  twistLimitDegrees=45
  twistMotion=Locked
 }
 {
  bone1Name=dyn_headHairParent
  bone2Name=dyn_torso
  swing1LimitDegrees=45
  swing1Motion=Locked
  swing2LimitDegrees=45
  swing2Motion=Locked
  twistLimitDegrees=45
  twistMotion=Locked
 }
 {
  bone1Name=dyn_skirtLeft
  bone2Name=dyn_skirt
  swing1LimitDegrees=45
  swing1Motion=Free
  swing2LimitDegrees=45
  swing2Motion=Free
  twistLimitDegrees=45
  twistMotion=Free
 }
 {
  bone1Name=dyn_skirtFront
  bone2Name=dyn_skirt
  swing1LimitDegrees=45
  swing1Motion=Free
  swing2LimitDegrees=45
  swing2Motion=Free
  twistLimitDegrees=45
  twistMotion=Free
 }
 {
  bone1Name=dyn_skirtBack
  bone2Name=dyn_skirt
  swing1LimitDegrees=45
  swing1Motion=Free
  swing2LimitDegrees=45
  swing2Motion=Free
  twistLimitDegrees=45
  twistMotion=Free
 }
 {
  bone1Name=dyn_skirtRight
  bone2Name=dyn_skirt
  swing1LimitDegrees=45
  swing1Motion=Free
  swing2LimitDegrees=45
  swing2Motion=Free
  twistLimitDegrees=45
  twistMotion=Free
 }
 {bone1Name=dyn_skirt, bone2Name=dyn_torso, swing1LimitDegrees=45, swing2LimitDegrees=45, twistLimitDegrees=45, twistMotion=Locked}
 {
  bone1Name=dyn_dreadlock_02_B
  bone2Name=dyn_headHairParent
  swing1LimitDegrees=15
  swing2LimitDegrees=25
  twistLimitDegrees=45
  twistMotion=Locked
 }
 {
  bone1Name=dyn_dreadlock_02_C
  bone2Name=dyn_dreadlock_02_B
  swing1LimitDegrees=15
  swing2LimitDegrees=25
  twistLimitDegrees=45
  twistMotion=Locked
 }
 {
  bone1Name=dyn_dreadlock_02_D
  bone2Name=dyn_dreadlock_02_C
  swing1LimitDegrees=15
  swing2LimitDegrees=25
  twistLimitDegrees=45
  twistMotion=Locked
 }
 {
  bone1Name=dyn_dreadlock_11_B
  bone2Name=dyn_headHairParent
  swing1LimitDegrees=25
  swing2LimitDegrees=25
  twistLimitDegrees=45
  twistMotion=Locked
 }
 {
  bone1Name=dyn_dreadlock_11_C
  bone2Name=dyn_dreadlock_11_B
  swing1LimitDegrees=25
  swing2LimitDegrees=25
  twistLimitDegrees=45
  twistMotion=Locked
 }
]
