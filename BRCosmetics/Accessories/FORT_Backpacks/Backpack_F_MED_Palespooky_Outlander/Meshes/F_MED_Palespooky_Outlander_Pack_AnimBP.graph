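# Anim Blueprint for the Palespooky Outlander backpack cosmetic, adding procedural secondary motion to its head, hair and sleeve bones.
# The input pose runs through AnimDynamics on dyn_head, then a RigidBody node using the Hair physics asset, based on the pack bone.
# Six chained AnimDynamics nodes then sway the left and right sleeve bones with reduced gravity, heavy damping and side-specific angular limits.
# The EventGraph holds only unconnected stub nodes.
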AnimBlueprint /BRCosmetics/Accessories/FORT_Backpacks/Backpack_F_MED_Palespooky_Outlander/Meshes/F_MED_Palespooky_Outlander_Pack_AnimBP
target_skeleton=/Game/Accessories/FORT_Backpacks/Backpack_F_MED_Palespooky_Outlander/Meshes/F_MED_Palespooky_Outlander_Pack_Skeleton
graphs [
 AnimGraph {nodes [
   n1 = Anim Dynamics - Bone: dyn_head(ComponentPose=Local To Component(LocalPose=Input Pose()), Alpha=1, ExternalForce="0,0,0") {
    angularDampingOverride=0.99
    bGravityOverrideInSimSpace=true
    bOverrideAngularDamping=true
    alphaBoolBlend {blendOption=Linear}
    boundBone {boneName=dyn_head}
    componentAppliedLinearAccClamp=<1000 1000 1000>
    gravityOverride {z=1}
    physicsBodyDefinitions [{
      sphereCollisionRadius=0
      boundBone {boneName=dyn_head}
      boxExtents=<5 5 15>
      constraintSetup {angularTargetAxis=AxisZ, twistAxis=AxisZ, angularLimitsMax=<114.6497 10 10>, angularLimitsMin=<0 -10 -10>, angularTarget=<0 0 0>}
      localJointOffset=<0 -0.1 -0.5>
     }]
   }
   n2 = RigidBody(ComponentPose=@n1) {
    bClampLinearTranslationLimitToRefPose=true
    bOverrideWorldGravity=true
    overridePhysicsAsset=/Game/Accessories/FORT_Backpacks/Backpack_F_MED_Palespooky_Outlander/Meshes/F_MED_Palespooky_Outlander_Pack_Hair_Physics
    alphaBoolBlend {blendOption=Linear}
    baseBoneRef {boneName=pack}
    componentAppliedLinearAccClamp=<950 950 950>
    componentLinearAccScale=<0.75 0.75 0.75>
    overrideWorldGravity {z=-980}
   }
   n3 = Anim Dynamics - Bone: dyn_sleeve_l_A(ComponentPose=@n2, Alpha=1, ExternalForce="0,0,0") {
    angularDampingOverride=0.9999
    bGravityOverrideInSimSpace=true
    bOverrideAngularDamping=true
    gravityScale=0.25
    alphaBoolBlend {blendOption=Linear}
    boundBone {boneName=dyn_sleeve_l_A}
    componentLinearAccScale=<0.4 0.35 0.25>
    physicsBodyDefinitions [{
      sphereCollisionRadius=0
      boundBone {boneName=dyn_sleeve_l_A}
      boxExtents=<5 5 5>
      constraintSetup {angularLimitsMax=<0.2866 10 75.9554>, angularLimitsMin=<-17.4841 -45 -7.1656>, angularTarget=<0 0 0>}
      localJointOffset=<1 0 -0.3>
     }]
   }
   n4 = Anim Dynamics - Bone: dyn_sleeve_l_B(ComponentPose=@n3, Alpha=1, ExternalForce="0,0,0") {
    angularDampingOverride=0.9999
    bGravityOverrideInSimSpace=true
    bOverrideAngularDamping=true
    gravityScale=0.25
    alphaBoolBlend {blendOption=Linear}
    boundBone {boneName=dyn_sleeve_l_B}
    componentLinearAccScale=<0.4 0.35 0.25>
    physicsBodyDefinitions [{
      sphereCollisionRadius=0
      boundBone {boneName=dyn_sleeve_l_B}
      boxExtents=<5 5 5>
      constraintSetup {angularLimitsMax=<0.2866 -6 75.9554>, angularLimitsMin=<-17.4841 -45 -7.1656>, angularTarget=<0 0 0>}
      localJointOffset=<1 0 -0.3>
     }]
   }
   n5 = Anim Dynamics - Bone: dyn_sleeve_l_C(ComponentPose=@n4, Alpha=1, ExternalForce="0,0,0") {
    angularDampingOverride=0.9999
    bGravityOverrideInSimSpace=true
    bOverrideAngularDamping=true
    gravityScale=0.25
    alphaBoolBlend {blendOption=Linear}
    boundBone {boneName=dyn_sleeve_l_C}
    componentLinearAccScale=<0.4 0.35 0.25>
    physicsBodyDefinitions [{
      sphereCollisionRadius=0
      boundBone {boneName=dyn_sleeve_l_C}
      boxExtents=<5 5 5>
      constraintSetup {angularLimitsMax=<0.2866 -6 75.9554>, angularLimitsMin=<-17.4841 -45 -7.1656>, angularTarget=<0 0 0>}
      localJointOffset=<1 0 -0.3>
     }]
   }
   n6 = Anim Dynamics - Bone: dyn_sleeve_r_A(ComponentPose=@n5, Alpha=1, ExternalForce="0,0,0") {
    angularDampingOverride=0.9999
    bGravityOverrideInSimSpace=true
    bOverrideAngularDamping=true
    gravityScale=0.25
    alphaBoolBlend {blendOption=Linear}
    boundBone {boneName=dyn_sleeve_r_A}
    componentLinearAccScale=<0.4 0.35 0.25>
    physicsBodyDefinitions [{
      sphereCollisionRadius=0
      boundBone {boneName=dyn_sleeve_r_A}
      boxExtents=<5 5 5>
      constraintSetup {angularLimitsMax=<0.2866 10 3.7261>, angularLimitsMin=<-17.4841 -45 -65.637>, angularTarget=<0 0 0>}
      localJointOffset=<1 0 -0.3>
     }]
   }
   n7 = Anim Dynamics - Bone: dyn_sleeve_r_B(ComponentPose=@n6, Alpha=1, ExternalForce="0,0,0") {
    angularDampingOverride=0.9999
    bGravityOverrideInSimSpace=true
    bOverrideAngularDamping=true
    gravityScale=0.25
    alphaBoolBlend {blendOption=Linear}
    boundBone {boneName=dyn_sleeve_r_B}
    componentLinearAccScale=<0.4 0.35 0.25>
    physicsBodyDefinitions [{
      sphereCollisionRadius=0
      boundBone {boneName=dyn_sleeve_r_B}
      boxExtents=<5 5 5>
      constraintSetup {angularLimitsMax=<0.2866 -5 3>, angularLimitsMin=<-17.4841 -45 -65.637>, angularTarget=<0 0 0>}
      localJointOffset=<1 0 -0.3>
     }]
   }
   n8 = Anim Dynamics - Bone: dyn_sleeve_r_C(ComponentPose=@n7, Alpha=1, ExternalForce="0,0,0") {
    angularDampingOverride=0.9999
    bGravityOverrideInSimSpace=true
    bOverrideAngularDamping=true
    gravityScale=0.25
    alphaBoolBlend {blendOption=Linear}
    boundBone {boneName=dyn_sleeve_r_C}
    componentLinearAccScale=<0.4 0.35 0.25>
    physicsBodyDefinitions [{
      sphereCollisionRadius=0
      boundBone {boneName=dyn_sleeve_r_C}
      boxExtents=<5 5 5>
      constraintSetup {angularLimitsMax=<0.2866 -5 3>, angularLimitsMin=<-17.4841 -45 -65.637>, angularTarget=<0 0 0>}
      localJointOffset=<1 0 -0.3>
     }]
   }
   Output Pose(Result=Component To Local(ComponentPose=@n8)) {layerGroup=None}
  ]}
 EventGraph {nodes [
   TryGetPawnOwner()
   Event BlueprintUpdateAnimation()
  ]}
]
variables []
